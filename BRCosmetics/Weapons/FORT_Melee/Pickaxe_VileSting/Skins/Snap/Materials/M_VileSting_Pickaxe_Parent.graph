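# Parent skinned-mesh material for the VileSting pickaxe skin, built on the 2023 character parent PreFX and PostFX functions with a Crew-style galaxy layer.
# Galaxy emissive blends two hue shifts along a rotatable, invertible position gradient, is desaturated, then gains masked cyan Fresnel before PostFX.
# A carbon-fiber function branches off the galaxy result to feed the clear coat normal, blending in fiber normals when isGold is on.
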Material /BRCosmetics/Weapons/FORT_Melee/Pickaxe_VileSting/Skins/Snap/Materials/M_VileSting_Pickaxe_Parent
graph {
 properties {
  bUseMaterialAttributes=true
  bUsedWithClothing=true
  bUsedWithMeshDeformer=true
  bUsedWithMorphTargets=true
  bUsedWithSkeletalMesh=true
  pixelDepthOffsetMode=PDOM_Legacy
  shadingModel=MSM_FromMaterialExpression
  subsurfaceProfile=/Game/Packages/Fortress_SharedMaterials/SubSurfaceProfile/Character_SubsurfaceProfile
  usedShadingModels=Default Lit | Subsurface | Clear Coat | Subsurface Profile | Cloth
 }
 nodes [
  n1 = MaterialFunctionCall() {materialFunction=&MF_CharacterParent_2023_PreFX}
  n2 = MaterialFunctionCall!(Tech Art Mask=NamedRerouteUsage(), Mask=NamedRerouteUsage(), In=@n1.Result) {materialFunction=/Game/Global/Materials/Character_Parent_2023/CrewStyle_Parent_2023/Galaxy/MF_CrewStyle_Galaxy_Inst}
  n3 = MaterialFunctionCall!(Alpha=NamedRerouteUsage(), In=@n2.Result) {materialFunction=/Game/Global/MaterialFunctions/TechArt/MF_CarbonFiberDualLayer}
  n4 = GetMaterialAttributes(None=@n3.Result) {attributeGetTypes [#1]}
  n5 = MaterialFunctionCall(
   BaseNormal=@n4.Normal
   AdditionalNormal=TextureSample(UVs=@n3.CarbonFiberUvs) {samplerType=SAMPLERTYPE_Normal, texture=/Game/Weapons/WeaponSkins/Wraps/Textures/T_Wrap_CarbonFiber_N}.RGB
  ) {materialFunction=&BlendAngleCorrectedNormals}
  ClearCoatNormalCustomOutput!(None=StaticSwitchParameter(True=LinearInterpolate(A=@n4.Normal, B=@n5.Result, Alpha=@n3.Mask), False=@n4.Normal) {parameterName=isGold})
  n6 = GetMaterialAttributes(None=@n2.Result) {attributeGetTypes [#2]}
  n7 = Power(Base=@n1.Fresnel)
  n8 = Constant3Vector() {constant=<rgba 0 0.6281 1 1>}
  n9 = NamedRerouteUsage()
  n10 = Multiply(
   A=Multiply(A=Multiply(A=@n7, B=@n8), B=ComponentMask(None=@n9) {r=true})
   B=ScalarParameter() {group=Galaxy_AdditionalControl, parameterName=Galaxy_AdditionalFresnelIntensity}
  )
  n11 = ScalarParameter() {group=Galaxy_AdditionalControl, parameterName=Galaxy_HueShift}
  n12 = ScalarParameter() {defaultValue=0.1, group=Galaxy_AdditionalControl, parameterName=Galaxy_PurpleHueShift}
  n13 = NamedRerouteUsage()
  n14 = ComponentMask(None=Rotator(Coordinate=ComponentMask(None=@n13) {b=true, g=true}, Time=ScalarParameter() {group=PurpleGradient, parameterName=PurpleGradient_Rotation})) {g=true}
  n15 = ScalarParameter() {defaultValue=-12, group=PurpleGradient, parameterName=PurpleGradient_PositionOffset}
  n16 = Saturate(None=Divide(A=Add(A=@n14, B=@n15) {constB=-10}, B=ScalarParameter() {defaultValue=40, group=PurpleGradient, parameterName=PurpleGradient_Smoothness}) {constB=40})
  n17 = OneMinus(None=@n16)
  n18 = ScalarParameter() {group=PurpleGradient, parameterName=PurpleGradient_InvertGradient}
  n19 = Desaturation(
   None=LinearInterpolate(
    A=MaterialFunctionCall(Hue Shift Percentage=@n11, Texture=@n6.Emissive Color) {materialFunction=&HueShift}.Result
    B=MaterialFunctionCall(Hue Shift Percentage=@n12, Texture=@n6.Emissive Color) {materialFunction=&HueShift}.Result
    Alpha=LinearInterpolate(A=@n16, B=@n17, Alpha=@n18)
   ) {constAlpha=1}
   Fraction=ScalarParameter() {defaultValue=-0.5, group=Galaxy_AdditionalControl, parameterName=Galaxy_Desaturation}
  )
  MaterialFunctionCall(
   Material=SetMaterialAttributes(MaterialAttributes=@n6.MaterialAttributes, Emissive Color=Add(A=@n10, B=@n19)) {attributeSetTypes [#2]}
   M=@n1.M
   Pre Skinned Local Position=@n1.M
   Distance Blend=@n1.M
   Fresnel=@n1.M
  ) {materialFunction=&MF_CharacterParent_2023_PostFX}
  NamedRerouteDeclaration(None=AppendVector(A=ComponentMask(None=@n1.M) {g=true, r=true}, B=Constant())) {name=Mask, nodeColor=<rgba 0 1 1 1>}
  NamedRerouteDeclaration(None=@n1.Tech Art Mask) {name=TechArtMask, nodeColor=<rgba 1 0.6797 0 1>}
  NamedRerouteDeclaration(None=@n1.Fresnel) {name=Fresnel, nodeColor=<rgba 0 1 0.4375 1>}
  NamedRerouteDeclaration(None=@n1.Pre skinned Local Position) {name=PreskinnedLocalPos, nodeColor=<rgba 0.125 0 1 1>}
 ]
}
